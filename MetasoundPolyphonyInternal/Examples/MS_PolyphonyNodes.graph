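# MS_PolyphonyNodes is an example MetaSound patch in MetasoundPolyphonyInternal that holds one of each of several polyphony node types.
# It places four voice controllers (Parameters, Trigger, TriggerIndex, TriggerParameters), mono and stereo PolyphonyPatchPlayers, and merge nodes for voice control streams and voice parameters.
# The nodes are unconnected and the patch has no inputs or outputs, so it produces no audio.
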MetaSoundPatch /MetasoundPolyphonyInternal/Examples/MS_PolyphonyNodes
inputs []
outputs []
pages [Default {nodes [
   MetasoundPolyphonyInternalMergeVoiceControlStreams4_0 = "MetasoundPolyphonyInternal.MergeVoiceControlStreams(4)"() {also {version [0,1]}}
   MetasoundPolyphonyInternal.ParametersVoiceController() {also {version [0,1]}}
   MetasoundPolyphonyInternal.PolyphonyPatchPlayer:Mono() {also {version [0,1]}}
   MetasoundPolyphonyInternal.PolyphonyPatchPlayer:Stereo() {also {version [0,1]}}
   MetasoundPolyphonyInternal.TriggerIndexVoiceController() {also {version [0,1]}}
   MetasoundPolyphonyInternal.TriggerParametersVoiceController() {also {version [0,1]}}
   MetasoundPolyphonyInternal.TriggerVoiceController() {also {version [0,1]}}
   MetasoundPolyphonyInternal.VoiceParametersMerge8Inputs() {also {version [0,1]}}
  ]}]
